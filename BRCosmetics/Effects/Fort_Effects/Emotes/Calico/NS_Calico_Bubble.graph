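# Translucent, default-lit material for Niagara mesh particles, apparently a bubble effect in the Calico emote.
# It samples a daytime cubemap along the world-space reflection vector to fake environment reflections, then hue-shifts the color by a scalar parameter.
# The shifted color is multiplied by particle color alpha, so each particle scales its reflection strength.
# The graph shows no material output links, and a Fresnel node sits unconnected.
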Material /BRCosmetics/Effects/Fort_Effects/Emotes/Calico/NS_Calico_Bubble
graph {
 properties {bUsedWithNiagaraMeshParticles=true, blendMode=BLEND_Translucent, pixelDepthOffsetMode=PDOM_Legacy, usedShadingModels=Default Lit}
 nodes [
  Fresnel() {exponent=2}
  n1 = ReflectionVectorWS()
  Multiply(
   A=ParticleColor().A
   B=MaterialFunctionCall(
    Hue Shift Percentage=ScalarParameter() {defaultValue=0.2}
    Texture=TextureSample(UVs=@n1) {texture=/Game/FN_PublicTurntable/TurntableAssets/Cubemap_Apollo_Day}.RGB
   ) {materialFunction=&HueShift}.Result
  )
 ]
}
