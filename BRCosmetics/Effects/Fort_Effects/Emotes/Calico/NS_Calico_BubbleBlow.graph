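# Niagara effect for the Calico emote that blows a bubble and pops it into droplets.
# Bubbles is a mesh emitter living for User.BubbleTime, with curves that swell it past its final size and push it along X, plus slight spin and dynamic material parameters.
# Drips bursts sprites at that same BubbleTime from a sphere offset to the bubble's end point, flinging them out under gravity and drag as they shrink and scale with speed.
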NiagaraSystem /BRCosmetics/Effects/Fort_Effects/Emotes/Calico/NS_Calico_BubbleBlow
dependencies {
 usedDataInterfaces [&NiagaraDataInterfaceCurve, &NiagaraDataInterfaceVectorCurve]
 usedDynamicInputs [&Lerp_Float, &MakeFloatFromInt]
 usedModules [&SystemState]
}
emitters {
 Bubbles {
  emitterUpdateScript [
   &EmitterState()
   &SpawnBurst_Instantaneous(Use Spawn Probability=false, UseLoopCountLimit=false)
  ]
  particleSpawnScript [
   &InitializeParticle(
    Color=<rgba 0.4792 0.4792 0.4792 0.3>
    Lifetime=$User.BubbleTime
    Mesh Scale Mode=::NewEnumerator3
    Mesh Uniform Scale=0.75
    Sprite Size Mode=::NewEnumerator1
    Uniform Sprite Size Max=50
    Uniform Sprite Size Min=30
    UsePositionOffset=false
    Write Material Random=true
   )
   &InitialMeshOrientation(
    Mesh Orientation Mode=::NewEnumerator2
    Rotation=&RandomRangeVector(Minimum=<0 0 0>, Maximum=<0 0 1>)
    Use Rotation Vector=true
   )
  ]
  particleUpdateScript [
   &ParticleState()
   &UpdateMeshOrientation(Rotation Rate=&RandomRangeFloat(Minimum=-0.1, Maximum=0.1), Rotation Vector=<0 0 1>)
   &ScaleMeshSize(Scale Factor=&VectorFromFloat(Value=&FloatFromCurve(FloatCurve=DataInterface() {propertyValues {Curve {keys [
          {arriveTangent=8.4581, interpMode=RCIM_Cubic, leaveTangent=8.4581, tangentMode=RCTM_User, time=0, value=0}
          {arriveTangent=-3.6276, interpMode=RCIM_Cubic, leaveTangent=-3.6276, tangentMode=RCTM_User, time=0.3, value=1.5}
          {arriveTangent=-1.795, interpMode=RCIM_Cubic, leaveTangent=-1.795, tangentMode=RCTM_User, time=0.8, value=1.375}
          {arriveTangent=-1.2715, tangentMode=RCTM_User, time=1, value=1.375}
         ]}}})))
   /SetVariables(Particles.Position=&VectorFromCurve(VectorCurve=DataInterface() {propertyValues {
       XCurve {keys [
         {arriveTangent=54.4269, interpMode=RCIM_Cubic, leaveTangent=54.4269, tangentMode=RCTM_User, time=0, value=0}
         {arriveTangent=-5.9824, interpMode=RCIM_Cubic, leaveTangent=-5.9824, tangentMode=RCTM_User, time=0.7, value=9.5}
         <at 1 10>
        ]}
       YCurve {keys [<at 0 0>,<at 1 0>]}
       ZCurve {keys [<at 0 0 RCIM_Cubic>,<at 1 0>]}
      }})) {isSetParameters=true}
   &SolveForcesAndVelocity(Clamp Velocity=false, Limit Acceleration=false)
   &DynamicMaterialParameters(
    Index 3 Param 1=1
    Index 3 Param 2=0.75
    Index 3 Param 3=1
    Index 3 Param 4=&RandomRangeFloat(Minimum=0, Maximum=1)
    Module.Index3Param0Write=true
    Module.Index3Param1Write=true
    Module.Index3Param2Write=true
    Module.Index3Param3Write=true
    Write Parameter Index 0=false
    Write Parameter Index 3=true
   )
  ]
  rendererClasses [&NiagaraMeshRendererProperties]
  renderers [{rendererClass=&NiagaraMeshRendererProperties, rendererIndex=0}]
 }
 Drips {
  emitterUpdateScript [
   &EmitterState()
   &SpawnBurst_Instantaneous(Spawn Count=20, Spawn Time=$User.BubbleTime, Use Spawn Probability=false, UseLoopCountLimit=false)
  ]
  particleSpawnScript [
   &InitializeParticle(
    Lifetime Max=0.35
    Lifetime Min=0.1
    Lifetime Mode=::NewEnumerator1
    Mass Max=2
    Mass Mode=::NewEnumerator3
    Sprite Size Max=<2 2>
    Sprite Size Min=<1 1>
    Sprite Size Mode=::NewEnumerator2
    UsePositionOffset=false
    Write Material Random=true
   )
   &ShapeLocation(
    MaskSpawnGroup=false
    Offset=<10 0 0>
    Offset Coordinate Space=::Local
    Offset Mode=::NewEnumerator0
    Override Seed=false
    Sphere Radius=10
    Sphere Surface Distribution=1
    UseOwnerScale=true
   )
   &AddVelocity(
    Module.Use Falloff Distance=false
    Origin Offset=<10 0 0>
    Override Seed=false
    Velocity Mode=::NewEnumerator1
    Velocity Speed=&RandomRangeFloat(Minimum=100, Maximum=300)
   )
  ]
  particleUpdateScript [
   &ParticleState()
   &GravityForce()
   &Drag(Use Linear Drag=true, Use Rotational Drag=false)
   &ScaleColor(
    Scale Alpha=&FloatFromCurve(FloatCurve=DataInterface() {propertyValues {Curve {keys [<at -0.0046 1.0113 RCIM_Cubic>,<at 1.0011 0.002 RCIM_Cubic>]}}})
    ScaleA=true
    ScaleRGB=true
   ) {enabled=false}
   &SolveForcesAndVelocity(Clamp Velocity=false, Limit Acceleration=false)
   &ScaleSpriteSizeBySpeed(Max Scale Factor=<0.75 3>, Velocity Threshold=200)
   &ScaleSpriteSize(Uniform Curve Sprite Scale=DataInterface() {propertyValues {Curve {keys [
        <at 0 1 RCIM_Cubic>
        {arriveTangent=-1, interpMode=RCIM_Cubic, leaveTangent=-1, time=0.5, value=0.725}
        {arriveTangent=-1.6294, interpMode=RCIM_Cubic, leaveTangent=-1.6294, tangentMode=RCTM_User, time=1, value=0}
       ]}}})
  ]
  rendererClasses [&NiagaraSpriteRendererProperties]
  renderers [{rendererClass=&NiagaraSpriteRendererProperties}]
 }
}
summary {
 systemName=NS_Calico_BubbleBlow
 emitters [
  {emitterName=Bubbles, rendererClasses [&NiagaraMeshRendererProperties]}
  {emitterName=Drips, rendererClasses [&NiagaraSpriteRendererProperties]}
 ]
 userVariables [User.BubbleTime {type=NiagaraFloat, defaultValue=0.8, description=""}]
}
